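; EQS context for the SatelliteSalvage secondary mission event that supplies a single location to environment queries.
; It overrides ProvideSingleLocation to return the constant world origin, ignoring both the querier object and querier actor.
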(fn ProvideSingleLocation
  (params (QuerierObject :type Object%20Reference) (QuerierActor :type Actor%20Object%20Reference))
  (return "0, 0, 0"))
